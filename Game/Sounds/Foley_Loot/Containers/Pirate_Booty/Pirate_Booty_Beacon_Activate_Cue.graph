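# Sound cue for the Pirate Booty loot beacon activation, on the loot reward sound class.
# A Mixer plays the beacon appear sound and a musical sting together, with the appear layer mixed lower.
# Attenuation overrides supersede the referenced item-interact asset and turn off volume falloff.
# A concurrency override limits the cue to one instance at a time.
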SoundCue /Game/Sounds/Foley_Loot/Containers/Pirate_Booty/Pirate_Booty_Beacon_Activate_Cue
root=Mixer_0
nodes [
 {id=Mixer_0, children [WavePlayer_1,WavePlayer_0], properties {inputVolume [0.7,1]}}
 {id=WavePlayer_0, properties {soundWaveAssetPtr=&/Pirate_Booty_Beacon_Appear_MX_Sting}}
 {id=WavePlayer_1, properties {soundWaveAssetPtr=&/Pirate_Booty_Beacon_Appear}}
]
properties {
 attenuationSettings=/Game/Sounds/Attenuation/Fort_Item_Interact_Att
 bOverrideAttenuation=true
 bOverrideConcurrency=true
 maxDistance=2097152
 soundClassObject=&Fort_Loot_Reward_PSM
 volumeMultiplier=1
 attenuationOverrides {bAttenuate=false, falloffDistance=100000, nonSpatializedRadiusStart=200, stereoSpread=1000, attenuationShapeExtents {x=5000}}
 concurrencyOverrides {maxCount=1}
}
