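# Sound cue for dig impacts on a Pirate Booty chest, filed under container loot foley.
# A Random node picks one of six impact wave variations with equal weight.
# It borrows small melee weapon attenuation and the melee weapon sound class, at reduced volume.
# A concurrency override caps simultaneous instances at three, stopping the oldest first.
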SoundCue /Game/Sounds/Foley_Loot/Containers/Pirate_Booty/Pirate_Booty_Chest_Dig_Impact_Cue
root=Random_1
nodes [
 {
  id=Random_1
  children [WavePlayer_2, WavePlayer_5, WavePlayer_6, WavePlayer_7, WavePlayer_8, WavePlayer_13]
  properties {weights [1,1,1,1,1,1]}
 }
 {id=WavePlayer_13, properties {soundWaveAssetPtr=&/Pirate_Booty_Chest_Dig_Impact_06}}
 {id=WavePlayer_2, properties {soundWaveAssetPtr=&/Pirate_Booty_Chest_Dig_Impact_01}}
 {id=WavePlayer_5, properties {soundWaveAssetPtr=&/Pirate_Booty_Chest_Dig_Impact_02}}
 {id=WavePlayer_6, properties {soundWaveAssetPtr=&/Pirate_Booty_Chest_Dig_Impact_03}}
 {id=WavePlayer_7, properties {soundWaveAssetPtr=&/Pirate_Booty_Chest_Dig_Impact_04}}
 {id=WavePlayer_8, properties {soundWaveAssetPtr=&/Pirate_Booty_Chest_Dig_Impact_05}}
]
properties {
 attenuationSettings=&Fort_Weapon_Melee_Sml_Att
 bOverrideConcurrency=true
 maxDistance=2800
 soundClassObject=&Fort_WeaponMelee_PSM
 volumeMultiplier=0.7
 concurrencyOverrides {maxCount=3, resolutionRule=StopOldest}
}
