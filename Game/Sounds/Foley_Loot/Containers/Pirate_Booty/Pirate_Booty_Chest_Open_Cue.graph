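# Sound cue for opening a Pirate Booty loot chest.
# A mixer plays the chest-open sound and a music sting wave together at equal volume.
# It uses the item-interact attenuation and the loot reward sound class, with a volume boost.
# A concurrency override prevents new instances from starting once the limit is reached.
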SoundCue /Game/Sounds/Foley_Loot/Containers/Pirate_Booty/Pirate_Booty_Chest_Open_Cue
root=Mixer_0
nodes [
 {id=Mixer_0, children [WavePlayer_0,WavePlayer_1], properties {inputVolume [1,1]}}
 {id=WavePlayer_0, properties {soundWaveAssetPtr=&/Pirate_Booty_Chest_Open_01}}
 {id=WavePlayer_1, properties {soundWaveAssetPtr=&/Pirate_Booty_Chest_Open_MX_Sting}}
]
properties {
 attenuationSettings=/Game/Sounds/Attenuation/Fort_Item_Interact_Att
 bOverrideConcurrency=true
 maxDistance=1850
 soundClassObject=&Fort_Loot_Reward_PSM
 volumeMultiplier=1.5
 concurrencyOverrides {resolutionRule=PreventNew}
}
