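# Sound cue for dropping the Pirate Booty loot container item, set in the Fort_Loot_Collect sound class.
# A single wave player feeds the root directly with the Pirate_Booty_Map_UnEquip_03 wave, with no randomization or modulation nodes.
# The cue uses weapon-drop attenuation at reduced volume, and its concurrency override allows one instance at a time, blocking new plays while one is active.
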SoundCue /Game/Sounds/Foley_Loot/Containers/Pirate_Booty/Pirate_Booty_Drop_Cue
root=WavePlayer_0
nodes [{id=WavePlayer_0, properties {soundWaveAssetPtr=&/Pirate_Booty_Map_UnEquip_03}}]
properties {
 attenuationSettings=&Fort_Weapon_Drop_Att
 bOverrideConcurrency=true
 maxDistance=3000
 soundClassObject=&Fort_Loot_Collect
 volumeMultiplier=0.478
 concurrencyOverrides {maxCount=1, resolutionRule=PreventNew}
}
