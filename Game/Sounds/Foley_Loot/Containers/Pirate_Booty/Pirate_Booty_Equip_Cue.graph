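# Sound cue for equipping the Pirate Booty map, part of the loot container foley.
# A Random node picks one of three equally weighted map-equip wave variants on each play.
# It routes through the loot-collect sound class with weapon-reload attenuation and lowered volume.
# A concurrency override allows one instance at a time and blocks new plays while one is active.
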SoundCue /Game/Sounds/Foley_Loot/Containers/Pirate_Booty/Pirate_Booty_Equip_Cue
root=Random_0
nodes [
 {id=Random_0, children [WavePlayer_1,WavePlayer_2,WavePlayer_3], properties {weights [1,1,1]}}
 {id=WavePlayer_1, properties {soundWaveAssetPtr=&/Pirate_Booty_Map_Equip_04}}
 {id=WavePlayer_2, properties {soundWaveAssetPtr=&/Pirate_Booty_Map_Equip_02}}
 {id=WavePlayer_3, properties {soundWaveAssetPtr=&/Pirate_Booty_Map_Equip_01}}
]
properties {
 attenuationSettings=&Fort_Weapon_Reload_Att
 bOverrideConcurrency=true
 maxDistance=2250
 soundClassObject=&Fort_Loot_Collect
 volumeMultiplier=0.597
 concurrencyOverrides {maxCount=1, resolutionRule=PreventNew}
}
