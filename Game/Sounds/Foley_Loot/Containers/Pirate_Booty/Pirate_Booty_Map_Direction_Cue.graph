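# Sound cue for the Pirate Booty map direction cue, played on the in-game non-critical UI sound class.
# A mixer layers a boosted frontend map-hover UI sound with a quieter paper direction-indicator sound, which an enveloper fades in over roughly the first third of a second.
# A modulator at the root pins volume while leaving default random pitch variation.
# A concurrency override allows one instance at a time, stopping the oldest.
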SoundCue /Game/Sounds/Foley_Loot/Containers/Pirate_Booty/Pirate_Booty_Map_Direction_Cue
root=Modulator_0
nodes [
 {
  id=Enveloper_0
  children [WavePlayer_3]
  properties {
   pitchCurve {editorCurveData {keys [{time=0}]}}
   volumeCurve {editorCurveData {keys [<at -0.0001 -0.0089>,<at 0 0.0028>,<at 0.1621 0.0871>,<at 0.3 1>]}}
  }
 }
 {id=Mixer_0, children [WavePlayer_0,Enveloper_0], properties {inputVolume [2,0.5]}}
 {id=Modulator_0, children [Mixer_0], properties {volumeMax=1, volumeMin=1}}
 {id=WavePlayer_0, properties {soundWaveAssetPtr=/Game/Sounds/Fort_UI_Sounds/FrontEnd/New_FrontEnd/Fortnite_UI_Map_Hover_01}}
 {id=WavePlayer_3, properties {soundWaveAssetPtr=&/Pirate_Booty_Paper_Direction_Indicator}}
]
properties {
 bOverrideConcurrency=true
 maxDistance=2097152
 soundClassObject=&Fort_UI_InGame_NonCritical
 volumeMultiplier=1.25
 concurrencyOverrides {maxCount=1, resolutionRule=StopOldest}
}
