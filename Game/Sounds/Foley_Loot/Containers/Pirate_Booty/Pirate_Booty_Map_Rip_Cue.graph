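# Sound cue for the map rip foley of the Pirate Booty loot container.
# A single wave player feeds the Pirate_Booty_Map_Rip sound wave straight to the output, with a boosted volume.
# It uses the item interaction attenuation and the loot reward sound class.
# A concurrency override blocks new instances once the limit is reached.
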SoundCue /Game/Sounds/Foley_Loot/Containers/Pirate_Booty/Pirate_Booty_Map_Rip_Cue
root=WavePlayer_2
nodes [{id=WavePlayer_2, properties {soundWaveAssetPtr=&/Pirate_Booty_Map_Rip}}]
properties {
 attenuationSettings=/Game/Sounds/Attenuation/Fort_Item_Interact_Att
 bOverrideConcurrency=true
 maxDistance=1850
 soundClassObject=&Fort_Loot_Reward_PSM
 volumeMultiplier=2
 concurrencyOverrides {resolutionRule=PreventNew}
}
