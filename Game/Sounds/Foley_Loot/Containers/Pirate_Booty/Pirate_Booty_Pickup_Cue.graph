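# Sound cue for picking up Pirate Booty loot containers in the game.
# A single wave player plays the map equip sound, with no randomization or modulation.
# It uses weapon pickup attenuation and the loot collect sound class at slightly reduced volume.
# A concurrency override allows one instance at a time and blocks new plays while one is active.
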SoundCue /Game/Sounds/Foley_Loot/Containers/Pirate_Booty/Pirate_Booty_Pickup_Cue
root=WavePlayer_3
nodes [{id=WavePlayer_3, properties {soundWaveAssetPtr=&/Pirate_Booty_Map_Equip_03}}]
properties {
 attenuationSettings=&Fort_Weapon_Pickup_Att
 bOverrideConcurrency=true
 maxDistance=3500
 soundClassObject=&Fort_Loot_Collect
 volumeMultiplier=0.796
 concurrencyOverrides {maxCount=1, resolutionRule=PreventNew}
}
