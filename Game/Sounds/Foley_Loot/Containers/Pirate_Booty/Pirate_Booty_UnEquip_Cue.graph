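# Sound cue for unequipping Pirate Booty loot, playing a map unequip foley sound.
# A Random node picks one of four equally weighted wave variations.
# The cue uses weapon reload attenuation and the loot collect sound class at a reduced volume.
# A concurrency override allows one instance at a time and prevents new plays while it is active.
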SoundCue /Game/Sounds/Foley_Loot/Containers/Pirate_Booty/Pirate_Booty_UnEquip_Cue
root=Random_1
nodes [
 {id=Random_1, children [WavePlayer_0,WavePlayer_4,WavePlayer_5,WavePlayer_6], properties {weights [1,1,1,1]}}
 {id=WavePlayer_0, properties {soundWaveAssetPtr=&/Pirate_Booty_Map_UnEquip_01}}
 {id=WavePlayer_4, properties {soundWaveAssetPtr=&/Pirate_Booty_Map_UnEquip_02}}
 {id=WavePlayer_5, properties {soundWaveAssetPtr=&/Pirate_Booty_Map_UnEquip_03}}
 {id=WavePlayer_6, properties {soundWaveAssetPtr=&/Pirate_Booty_Map_UnEquip_04}}
]
properties {
 attenuationSettings=&Fort_Weapon_Reload_Att
 bOverrideConcurrency=true
 maxDistance=2250
 soundClassObject=&Fort_Loot_Collect
 volumeMultiplier=0.597
 concurrencyOverrides {maxCount=1, resolutionRule=PreventNew}
}
